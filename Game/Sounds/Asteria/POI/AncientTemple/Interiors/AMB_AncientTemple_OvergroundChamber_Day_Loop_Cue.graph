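# Looping daytime interior ambience cue for the overground chamber of the Asteria Ancient Temple POI.
# A Quality Level root routes its first two slots to a looping quad wave and its last two to a stereo version.
# Each wave runs through an indefinitely looping 30-second Enveloper with flat volume and a cubic pitch curve.
# Indoor ambient sound class, ambient attenuation and an owner-limited single-instance concurrency override govern playback.
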SoundCue /Game/Sounds/Asteria/POI/AncientTemple/Interiors/AMB_AncientTemple_OvergroundChamber_Day_Loop_Cue
root=QualityLevel_0
nodes [
 {
  id=Enveloper_1
  children [WavePlayer_0]
  properties {
   bLoop=true
   bLoopIndefinitely=true
   loopEnd=30
   pitchCurve {editorCurveData {keys [{interpMode=RCIM_Cubic, time=0}, <at 15.9468 0.9181 RCIM_Cubic>, {interpMode=RCIM_Cubic, time=30}]}}
   volumeCurve {editorCurveData {keys [<at 0 1>]}}
  }
 }
 {
  id=Enveloper_4
  children [WavePlayer_1]
  properties {
   bLoop=true
   bLoopIndefinitely=true
   loopEnd=30
   pitchCurve {editorCurveData {keys [{interpMode=RCIM_Cubic, time=0}, <at 15.9468 0.9181 RCIM_Cubic>, {interpMode=RCIM_Cubic, time=30}]}}
   volumeCurve {editorCurveData {keys [<at 0 1>]}}
  }
 }
 {id=QualityLevel_0, children [Enveloper_1,Enveloper_1,Enveloper_4,Enveloper_4]}
 {
  id=WavePlayer_0
  properties {bLooping=true, soundWaveAssetPtr=&/AMB_AncientTemple_OvergroundChamber_Day_Loop_Quad}
 }
 {
  id=WavePlayer_1
  properties {bLooping=true, soundWaveAssetPtr=&/AMB_AncientTemple_OvergroundChamber_Day_Loop_Stereo}
 }
]
properties {
 attenuationSettings=&Fort_Ambient_Athena_Att
 bOverrideConcurrency=true
 maxDistance=2097152
 soundClassObject=&Fort_Ambient_Indoors_Athena
 volumeMultiplier=0.376
 concurrencyOverrides {bLimitToOwner=true, maxCount=1, resolutionRule=PreventNew, retriggerTime=2}
}
