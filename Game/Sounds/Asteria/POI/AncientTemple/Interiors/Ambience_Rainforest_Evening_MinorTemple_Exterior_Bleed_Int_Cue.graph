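# Looping sound cue for evening rainforest exterior ambience, without birds or trees, bleeding into Ancient Temple minor temple interiors.
# A quality node selects the quad loop for its first two levels and the stereo loop for the last two.
# Each loop runs through an endlessly repeating envelope that gently dips and restores volume, then a mixer boost.
# It uses the indoor ambient sound class, ambient attenuation and one-instance-per-owner concurrency that blocks new plays.
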SoundCue /Game/Sounds/Asteria/POI/AncientTemple/Interiors/Ambience_Rainforest_Evening_MinorTemple_Exterior_Bleed_Int_Cue
root=QualityLevel_0
nodes [
 {
  id=Enveloper_5
  children [WavePlayer_5]
  properties {
   bLoop=true
   bLoopIndefinitely=true
   loopEnd=30
   pitchCurve {editorCurveData {keys [{time=0}]}}
   volumeCurve {editorCurveData {keys [
      <at 0 1 RCIM_Cubic>
      {arriveTangent=0, interpMode=RCIM_Cubic, leaveTangent=0, tangentMode=RCTM_User, time=15, value=0.9}
      <at 30 1 RCIM_Cubic>
     ]}}
  }
 }
 {
  id=Enveloper_6
  children [WavePlayer_4]
  properties {
   bLoop=true
   bLoopIndefinitely=true
   loopEnd=30
   pitchCurve {editorCurveData {keys [{time=0}]}}
   volumeCurve {editorCurveData {keys [
      <at 0 1 RCIM_Cubic>
      {arriveTangent=0, interpMode=RCIM_Cubic, leaveTangent=0, tangentMode=RCTM_User, time=15, value=0.9}
      <at 30 1 RCIM_Cubic>
     ]}}
  }
 }
 {id=Mixer_0, children [Enveloper_6], properties {inputVolume [1.5]}}
 {id=Mixer_1, children [Enveloper_5], properties {inputVolume [1.5]}}
 {id=QualityLevel_0, children [Mixer_0,Mixer_0,Mixer_1,Mixer_1]}
 {
  id=WavePlayer_4
  properties {bLooping=true, soundWaveAssetPtr=&/Ambience_Rainforest_Evening_NoBirds_NoTrees_Loop_Quad}
 }
 {
  id=WavePlayer_5
  properties {bLooping=true, soundWaveAssetPtr=&/Ambience_Rainforest_Evening_NoBirds_NoTrees_Loop_Stereo}
 }
]
properties {
 attenuationSettings=&Fort_Ambient_Athena_Att
 bOverrideConcurrency=true
 maxDistance=2097152
 soundClassObject=&Fort_Ambient_Indoors_Athena
 volumeMultiplier=0.401
 concurrencyOverrides {bLimitToOwner=true, maxCount=1, resolutionRule=PreventNew, retriggerTime=2}
}
