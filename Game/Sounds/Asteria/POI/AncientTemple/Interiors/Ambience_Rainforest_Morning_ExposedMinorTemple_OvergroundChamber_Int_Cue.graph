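# Looping morning rainforest ambience cue for the interior overground chamber of an exposed minor ancient temple, classed as indoor ambient.
# A quality-level node sends the first two levels to a quad mix and the last two to a stereo mix.
# Each mix layers a temple interior loop with two morning exterior beds, weird birds and treeless rainforest, whose looping envelopes gently dip and recover in volume.
# A concurrency override allows one instance per owner, preventing new starts.
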SoundCue /Game/Sounds/Asteria/POI/AncientTemple/Interiors/Ambience_Rainforest_Morning_ExposedMinorTemple_OvergroundChamber_Int_Cue
root=QualityLevel_0
nodes [
 {
  id=Enveloper_0
  children [WavePlayer_3]
  properties {
   bLoop=true
   bLoopIndefinitely=true
   loopEnd=30
   pitchCurve {editorCurveData {keys [{time=0}]}}
   volumeCurve {editorCurveData {keys [
      <at 0 1 RCIM_Cubic>
      {arriveTangent=0, interpMode=RCIM_Cubic, leaveTangent=0, tangentMode=RCTM_User, time=15, value=0.9}
      <at 30 1 RCIM_Cubic>
     ]}}
  }
 }
 {
  id=Enveloper_1
  children [WavePlayer_2]
  properties {
   bLoop=true
   bLoopIndefinitely=true
   loopEnd=30
   pitchCurve {editorCurveData {keys [{time=0}]}}
   volumeCurve {editorCurveData {keys [
      <at 0 1 RCIM_Cubic>
      {arriveTangent=0, interpMode=RCIM_Cubic, leaveTangent=0, tangentMode=RCTM_User, time=15, value=0.9}
      <at 30 1 RCIM_Cubic>
     ]}}
  }
 }
 {
  id=Enveloper_5
  children [WavePlayer_6]
  properties {
   bLoop=true
   bLoopIndefinitely=true
   loopEnd=30
   pitchCurve {editorCurveData {keys [{time=0}]}}
   volumeCurve {editorCurveData {keys [
      <at 0 1 RCIM_Cubic>
      {arriveTangent=0, interpMode=RCIM_Cubic, leaveTangent=0, tangentMode=RCTM_User, time=15, value=0.9}
      <at 30 1 RCIM_Cubic>
     ]}}
  }
 }
 {
  id=Enveloper_6
  children [WavePlayer_7]
  properties {
   bLoop=true
   bLoopIndefinitely=true
   loopEnd=30
   pitchCurve {editorCurveData {keys [{time=0}]}}
   volumeCurve {editorCurveData {keys [
      <at 0 1 RCIM_Cubic>
      {arriveTangent=0, interpMode=RCIM_Cubic, leaveTangent=0, tangentMode=RCTM_User, time=15, value=0.9}
      <at 30 1 RCIM_Cubic>
     ]}}
  }
 }
 {id=Mixer_0, children [Enveloper_1,Enveloper_6,WavePlayer_0], properties {inputVolume [1.5,0.6,1.5]}}
 {id=Mixer_1, children [WavePlayer_1,Enveloper_5,Enveloper_0], properties {inputVolume [1.5,0.6,1.5]}}
 {id=QualityLevel_0, children [Mixer_0,Mixer_0,Mixer_1,Mixer_1]}
 {
  id=WavePlayer_0
  properties {bLooping=true, soundWaveAssetPtr=/Game/Sounds/Artemis/POI/Jungle_Temple/AMB_Canary_Jungle_Temple_Int_01_Loop_Quad}
 }
 {
  id=WavePlayer_1
  properties {bLooping=true, soundWaveAssetPtr=/Game/Sounds/Artemis/POI/Jungle_Temple/AMB_Canary_Jungle_Temple_Int_01_Loop_ST}
 }
 {
  id=WavePlayer_2
  properties {
   bLooping=true
   soundWaveAssetPtr=&/Ambience_Rainforest_AncientTemple_Morning_Weird_Birds_Ext_Loop_Quad
  }
 }
 {
  id=WavePlayer_3
  properties {
   bLooping=true
   soundWaveAssetPtr=&/Ambience_Rainforest_AncientTemple_Morning_Weird_Birds_Ext_Loop_Stereo
  }
 }
 {
  id=WavePlayer_6
  properties {bLooping=true, soundWaveAssetPtr=&/Ambience_Rainforest_Morning_NoTrees_Loop_Stereo}
 }
 {
  id=WavePlayer_7
  properties {bLooping=true, soundWaveAssetPtr=&/Ambience_Rainforest_Morning_NoTrees_Loop_Quad}
 }
]
properties {
 attenuationSettings=&Fort_Ambient_Athena_Att
 bOverrideConcurrency=true
 maxDistance=2097152
 soundClassObject=&Fort_Ambient_Indoors_Athena
 volumeMultiplier=0.276
 concurrencyOverrides {bLimitToOwner=true, maxCount=1, resolutionRule=PreventNew, retriggerTime=2}
}
