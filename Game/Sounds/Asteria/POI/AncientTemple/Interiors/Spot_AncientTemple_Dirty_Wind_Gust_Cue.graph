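# Sound cue for one-shot dirty wind gust ambience in Ancient Temple interiors, reusing Hollow Base cave gust waves.
# A quality switch sends its first two levels to a random pick of eleven gusts, and its last two to only the first four.
# Each branch adds random pitch and boosted volume, with the larger pool pitched lower.
# Playback uses the indoor ambient sound class and tile-ambient one-shot attenuation.
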SoundCue /Game/Sounds/Asteria/POI/AncientTemple/Interiors/Spot_AncientTemple_Dirty_Wind_Gust_Cue
root=QualityLevel_0
nodes [
 {id=Modulator_0, children [Random_1], properties {pitchMax=1, pitchMin=0.8, volumeMax=2.1, volumeMin=2}}
 {id=Modulator_2, children [Random_0], properties {pitchMax=1.1, volumeMax=2.2, volumeMin=2}}
 {id=QualityLevel_0, children [Modulator_0,Modulator_0,Modulator_2,Modulator_2]}
 {id=Random_0, children [WavePlayer_0,WavePlayer_1,WavePlayer_2,WavePlayer_3], properties {weights [1,1,1,1]}}
 {
  id=Random_1
  children [
   WavePlayer_5
   WavePlayer_6
   WavePlayer_7
   WavePlayer_8
   WavePlayer_9
   WavePlayer_10
   WavePlayer_11
   WavePlayer_12
   WavePlayer_13
   WavePlayer_14
   WavePlayer_15
  ]
  properties {weights [1,1,1,1,1,1,1,1,1,1,1]}
 }
 {id=WavePlayer_0, properties {soundWaveAssetPtr=/Game/Sounds/Artemis/POI/Hollow_Base/Spots/Spot_Cave_Dirty_Wind_Gust_01}}
 {id=WavePlayer_1, properties {soundWaveAssetPtr=/Game/Sounds/Artemis/POI/Hollow_Base/Spots/Spot_Cave_Dirty_Wind_Gust_02}}
 {id=WavePlayer_10, properties {soundWaveAssetPtr=/Game/Sounds/Artemis/POI/Hollow_Base/Spots/Spot_Cave_Dirty_Wind_Gust_06}}
 {id=WavePlayer_11, properties {soundWaveAssetPtr=/Game/Sounds/Artemis/POI/Hollow_Base/Spots/Spot_Cave_Dirty_Wind_Gust_07}}
 {id=WavePlayer_12, properties {soundWaveAssetPtr=/Game/Sounds/Artemis/POI/Hollow_Base/Spots/Spot_Cave_Dirty_Wind_Gust_08}}
 {id=WavePlayer_13, properties {soundWaveAssetPtr=/Game/Sounds/Artemis/POI/Hollow_Base/Spots/Spot_Cave_Dirty_Wind_Gust_09}}
 {id=WavePlayer_14, properties {soundWaveAssetPtr=/Game/Sounds/Artemis/POI/Hollow_Base/Spots/Spot_Cave_Dirty_Wind_Gust_10}}
 {id=WavePlayer_15, properties {soundWaveAssetPtr=/Game/Sounds/Artemis/POI/Hollow_Base/Spots/Spot_Cave_Dirty_Wind_Gust_11}}
 {id=WavePlayer_2, properties {soundWaveAssetPtr=/Game/Sounds/Artemis/POI/Hollow_Base/Spots/Spot_Cave_Dirty_Wind_Gust_03}}
 {id=WavePlayer_3, properties {soundWaveAssetPtr=/Game/Sounds/Artemis/POI/Hollow_Base/Spots/Spot_Cave_Dirty_Wind_Gust_04}}
 {id=WavePlayer_5, properties {soundWaveAssetPtr=&Spot_Cave_Dirty_Wind_Gust_01}}
 {id=WavePlayer_6, properties {soundWaveAssetPtr=&Spot_Cave_Dirty_Wind_Gust_03}}
 {id=WavePlayer_7, properties {soundWaveAssetPtr=/Game/Sounds/Artemis/POI/Hollow_Base/Spots/Spot_Cave_Dirty_Wind_Gust_05}}
 {id=WavePlayer_8, properties {soundWaveAssetPtr=&Spot_Cave_Dirty_Wind_Gust_02}}
 {id=WavePlayer_9, properties {soundWaveAssetPtr=&Spot_Cave_Dirty_Wind_Gust_04}}
]
properties {
 attenuationSettings=/Game/Sounds/Attenuation/Fort_TileAmbient_OneShot_Att
 maxDistance=8000
 soundClassObject=&Fort_Ambient_Indoors_Athena
 volumeMultiplier=0.251
 preEffectBusSends [{}]
}
